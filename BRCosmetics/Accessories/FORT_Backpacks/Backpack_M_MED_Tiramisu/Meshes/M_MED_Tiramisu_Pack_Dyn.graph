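# Physics asset for the Tiramisu cosmetic backpack, defining collision bodies and a joint for its sword bone.
# A large capsule covers the dyn_sword bone, and a much smaller capsule sits on the pack bone.
# A single constraint links the two bones with swing 1 locked and small limit angles set for swing 2 and twist.
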PhysicsAsset /BRCosmetics/Accessories/FORT_Backpacks/Backpack_M_MED_Tiramisu/Meshes/M_MED_Tiramisu_Pack_Dyn
bodies {
 dyn_sword [{shapeType=Capsule, length=44.7034, radius=8.7356, center=<0.7007 0 0>, rotation=<rot 90 0 0>}]
 pack [{shapeType=Capsule, length=4.7428, radius=0.5, center=<0.7469 -2.2272 -1.1321>, rotation=<rot 0.1472 -90 -0.0035>}]
}
constraints [{bone1Name=dyn_sword, bone2Name=pack, swing1LimitDegrees=45, swing1Motion=Locked, swing2LimitDegrees=10, twistLimitDegrees=1}]
