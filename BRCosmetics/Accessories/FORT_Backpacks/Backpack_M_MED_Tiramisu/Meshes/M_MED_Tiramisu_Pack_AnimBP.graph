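# Animation Blueprint for the Tiramisu backpack cosmetic, posing the pack and toggling a material-hidden element.
# Each update maps the owner mesh's spine_05 yaw, relative to spine_01 and ref pose, to additive roll and Y offset on the pack bone.
# A RigidBody node then simulates dynamics in root-bone space.
# Weapon, customization and skydiving events set "Hide Element 01", hiding it while a Tiramisu-tagged pickaxe is equipped outside skydiving.
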AnimBlueprint /BRCosmetics/Accessories/FORT_Backpacks/Backpack_M_MED_Tiramisu/Meshes/M_MED_Tiramisu_Pack_AnimBP
target_skeleton=&/M_MED_Tiramisu_Pack_Skeleton
graphs [
 AnimGraph {nodes [
   n1 = "Transform (Modify) Bone - Bone: pack"(ComponentPose=Local To Component(LocalPose=Local Space Ref Pose()), Translation=Get backpack_translation(), Rotation=Get backpack_rotate()) {rotationMode=BMM_Additive, translationMode=BMM_Additive, alphaBoolBlend {blendOption=Linear}, boneToModify {boneName=pack}}
   n2 = RigidBody!(ComponentPose=@n1, Alpha, OverrideWorldGravity:Vector="0,0,-980", ExternalForce) -> (Pose) {
    bClampLinearTranslationLimitToRefPose=true
    bOverrideWorldGravity=true
    overridePhysicsAsset=&/M_MED_Tiramisu_Pack_Dyn
    simulationSpace=BaseBoneSpace
    alphaBoolBlend {blendOption=Linear}
    baseBoneRef {boneName=root}
    componentAppliedLinearAccClamp=<3000 3000 3000>
    componentLinearAccScale {z=0.5}
    componentLinearVelScale {z=0.5}
    overrideWorldGravity {z=-980}
    simSpaceSettings {
     maxAngularAcceleration=1000
     maxAngularVelocity=3
     maxLinearAcceleration=700
     maxLinearVelocity=500
     velocityScaleZ=0
     worldAlpha=1
     externalLinearDragV=<1 1>
    }
   }
   Output Pose(Result=Component To Local(ComponentPose=@n2)) {layerGroup=None}
  ]}
 EventGraph {nodes [
   Event BlueprintUpdateAnimation()
   n1 = GetOwnerSkeletalMeshComponent()
   n2 = BreakRotator(InRot=BreakTransform(InTransform=GetDeltaTransformFromRefPose(self=@n1, BoneName=spine_05, BaseName=spine_01)).Rotation)
   > Set backpack_rotate(backpack_rotate=MakeRotator(Roll=MapRangeClamped(Value=@n2.Yaw, InRangeA=0, InRangeB=-30, OutRangeB=20)))
   > Set backpack_translation(backpack_translation=MakeVector(Y=MapRangeClamped(Value=@n2.Yaw, InRangeA=0, InRangeB=-30, OutRangeB=-10)))
   n3 = Event OnWeaponChanged()
   n4 = Get Dynamic Material Instance(execute=[@n3.then, @n12.then], SkeletalMeshComponent=GetPartSkeletalMeshComponent())
   > n5 = Sequence()
   n6 = PickaxeItemDefinitionHasAnyMetaTag(Tags=Make Array("[0]"=(TagName="Cosmetics.MetaTags.Tiramisu")), bRequiresEquipped=true)
   > n7 = Branch(Condition=@n6)
   > n8 = Branch(Condition=Get bIsSkyDiving())
   SetScalarParameterValue(execute=[@n10.then, @n8.then, @n9.else, @n7.else], self=@n4, ParameterName=Hide Element 01, Value=0)
   SetScalarParameterValue(execute=[@n8.else, @n9.then], self=@n4, ParameterName=Hide Element 01, Value=1)
   n9 = Branch(execute=[@n11.then, @n5.then_1], Condition=@n6)
   n10 = Event OnBeginSkydiving()
   n11 = Event OnEndSkydiving()
   n12 = Event OnCharacterCustomizationLoaded()
   TryGetPawnOwner()
  ]}
]
variables [
 backpack_rotate {type=Rotator}
 backpack_translation {type=Vector}
]
